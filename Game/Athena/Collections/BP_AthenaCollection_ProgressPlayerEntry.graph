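# Player marker widget for collection progress, showing a friend's avatar, name and collected/total count, or a compact arrow and count for the local player.
# Progress initialization picks the layout and name color, fills the rich-text counts, and flips the friend text's side past halfway.
# It also snaps ShowUp to its collapsed state; show-info events play it forward or reverse.
# Height adapts for mobile, and PreConstruct previews either layout at design time.
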WidgetBlueprint /Game/Athena/Collections/BP_AthenaCollection_ProgressPlayerEntry
animations [ShowUp {
  binding_count=2
  length_seconds=0.25
  object=&self:ShowUp
  sequence {bindings [
    Image_Arrow {
     id=#1
     display_name=Image_85
     possessed_class=CommonTextBlock
     tracks [
      {
       type=MovieScene2DTransformTrack
       sections [{
         type=MovieScene2DTransformSection
         channels [
          Translation.X {type=MovieSceneScriptingFloatChannel}
          Translation.Y {type=MovieSceneScriptingFloatChannel}
          Angle {type=MovieSceneScriptingFloatChannel}
          Scale.X {type=MovieSceneScriptingFloatChannel, keys [[0,0.75],[2,0.5],[5,1]]}
          Scale.Y {type=MovieSceneScriptingFloatChannel, keys [[0,0.75],[2,0.5],[5,1]]}
          Shear.X {type=MovieSceneScriptingFloatChannel}
          Shear.Y {type=MovieSceneScriptingFloatChannel}
         ]
        }]
      }
      {
       type=MovieSceneColorTrack
       display_name=Color and Opacity
       sections [{
         type=MovieSceneColorSection
         channels [
          Color.R {type=MovieSceneScriptingFloatChannel, keys [[0,0],[5.0003,1]]}
          Color.G {type=MovieSceneScriptingFloatChannel}
          Color.B {type=MovieSceneScriptingFloatChannel}
          Color.A {type=MovieSceneScriptingFloatChannel}
         ]
        }]
      }
     ]
    }
    Overlay_0 {
     id=#2
     possessed_class=Overlay
     tracks [
      {
       type=MovieScene2DTransformTrack
       sections [{
         type=MovieScene2DTransformSection
         channels [
          Translation.X {type=MovieSceneScriptingFloatChannel}
          Translation.Y {type=MovieSceneScriptingFloatChannel}
          Angle {type=MovieSceneScriptingFloatChannel}
          Scale.X {type=MovieSceneScriptingFloatChannel, keys [[0,0.5],[3,0.9],[5,1]]}
          Scale.Y {type=MovieSceneScriptingFloatChannel, keys [[0,0],[3,1.1],[5,1]]}
          Shear.X {type=MovieSceneScriptingFloatChannel}
          Shear.Y {type=MovieSceneScriptingFloatChannel}
         ]
        }]
      }
      {
       type=MovieSceneFloatTrack
       display_name=Render Opacity
       sections [{type=MovieSceneFloatSection, channels [None {type=MovieSceneScriptingFloatChannel, keys [[0,0],[3,1]]}]}]
      }
     ]
    }
   ]}
 }]
graphs [EventGraph {nodes [
   n1 = Event PreConstruct()
   > Branch(Condition=@n1.IsDesignTime)
   > SetActiveWidget(
    self:Widget Switcher Object Reference=Get WidgetSwitcher_0() -> (WidgetSwitcher_0:Widget Switcher Object Reference)
    Widget=Select(
     Option 0:Widget Object Reference=Get Overlay_Friends() -> (Overlay_Friends:Overlay Object Reference)
     Option 1:Widget Object Reference=Get SizeBox_LocalPlayer() -> (SizeBox_LocalPlayer:Size Box Object Reference)
     Index=Get Is Self pre construct() -> (Is Self pre construct:bool)
    ) -> (ReturnValue:Widget Object Reference)
   )
   n2 = Event On Player Progress Initialized!() -> (OutputDelegate, then, bIsLocalPlayer:bool, NumCollected:int=0, NumToCollect:int=0) {is=title K2Node_Event}
   > SetColorAndOpacity(
    self=Get Text_PlayerName() -> (Text_PlayerName:Common Text Object Reference)
    InColorAndOpacity=Select(
     Option 0:Slate Color Structure=Make Slate Color(SpecifiedColor=(R=0.0331,G=0.855,B=1,A=1))
     Option 1:Slate Color Structure=Make Slate Color(SpecifiedColor=(R=0.9301,G=1,B=0.0103,A=1))
     Index=@n2.bIsLocalPlayer
    ) -> (ReturnValue:Slate Color Structure)
   )
   > SetActiveWidget(
    self=Get WidgetSwitcher_0()
    Widget=Select(Option 0:Widget Object Reference=Get Overlay_Friends(), Option 1:Widget Object Reference=Get SizeBox_LocalPlayer(), Index=@n2.bIsLocalPlayer) -> (ReturnValue:Widget Object Reference)
   )
   n3 = Get ShowUp() -> (ShowUp:Widget Animation Object Reference)
   > PlayAnimation(InAnimation=@n3, StartAtTime=GetEndTime(self=@n3), PlayMode=Reverse)
   > SetText(
    self=Get RichText_Number() -> (RichText_Number:Common Rich Text Block Object Reference)
    InText=Format Text!(
     Format="INVTEXT(\"{current}<text color=\\\"00a6ffff\\\" size=\\\"14\\\">/{total}</>\")"
     current:int=@n2.NumCollected
     total:int=@n2.NumToCollect
    ) -> (Result)
   )
   > SetText(
    self=Get RichText_Number_Local() -> (RichText_Number_Local:Common Rich Text Block Object Reference)
    InText=Format Text(Format="INVTEXT(\"{current}<text color=\\\"00a6ffff\\\" size=\\\"12\\\">/{total}</>\")", current=@n2.NumCollected, total=@n2.NumToCollect)
   )
   n4 = "integer < integer"!(A:int=int / int!(A:int=@n2.NumToCollect, B:int=2) -> (ReturnValue:int=0) {is=title K2Node_CallFunction}, B:int=@n2.NumCollected) -> (ReturnValue:bool) {is=title K2Node_CallFunction}
   n5 = Select(Option 0:ETextJustify Enum=Left, Option 1:ETextJustify Enum=Right, Index=@n4) -> (ReturnValue:ETextJustify Enum)
   > Set Justification!(execute, Justification:ETextJustify Enum=@n5, self:Text Layout Widget Object Reference=Get Text_PlayerName()) -> (then, Output_Get:ETextJustify Enum=Left)
   > Set Justification(Justification=@n5, self=Get RichText_Number()) -> (Output_Get=Left)
   > SetRenderTranslation(
    self=Get VerticalBox_text() -> (VerticalBox_text:Vertical Box Object Reference)
    Translation=Select(Option 0:Vector 2D Structure=(X=32,Y=0), Option 1:Vector 2D Structure=(X=-32,Y=0), Index=@n4) -> (ReturnValue:Vector 2D Structure)
   )
   n6 = Event On Show Player Info Set!() -> (OutputDelegate, then, bShowPlayerInfo:bool) {is=title K2Node_Event}
   > PlayAnimation(
    InAnimation=Get ShowUp()
    PlayMode=Select(Option 0:EUMGSequencePlayMode Enum=Reverse, Option 1:EUMGSequencePlayMode Enum, Index=@n6.bShowPlayerInfo) -> (ReturnValue:EUMGSequencePlayMode Enum)
   )
   Event OnInitialized()
   > SetHeightOverride(self=Get SizeBox_0() -> (SizeBox_0:Size Box Object Reference), InHeightOverride=SelectFloat(A=140, B=116, bPickA=IsMobileGame()))
  ]}]
variables [Is Self pre construct {type=boolean}]
widgets {
 info {inheritedWidgetCount=3, parentClass=/Script/FortniteUI.AthenaCollectionPlayerProgress, rootWidgetClass=&SizeBox, widgetCount=20}
 widgets [
  SizeBox_0 {widgetClassPath=&SizeBox, bIsVariable=true, widget=SizeBox_0, properties {bOverride_HeightOverride=true, heightOverride=116}}
  WidgetSwitcher_0 {widgetClassPath=&WidgetSwitcher, bIsVariable=true, parent=SizeBox_0, widget=WidgetSwitcher_0, slot {class=SizeBoxSlot}}
  Overlay_Friends {
   widgetClassPath=&Overlay
   bIsVariable=true
   parent=WidgetSwitcher_0
   widget=Overlay_Friends
   slot {class=WidgetSwitcherSlot, properties {verticalAlignment=VAlign_Top}}
  }
  VerticalBox_17 {
   widgetClassPath=&VerticalBox
   parent=Overlay_Friends
   widget=VerticalBox_17
   slot {class=OverlaySlot, properties {horizontalAlignment=HAlign_Right, verticalAlignment=VAlign_Bottom, padding=<ltrb -23 0 -23 0>}}
  }
  Overlay_0 {
   widgetClassPath=&Overlay
   parent=VerticalBox_17
   widget=Overlay_0
   properties {renderTransformPivot=<0.5 1>}
   slot {class=VerticalBoxSlot, properties {horizontalAlignment=HAlign_Center, verticalAlignment=VAlign_Center, padding=<ltrb 0 0 0 -2>}}
  }
  AvatarIcon {
   widgetClassPath=/Script/FortniteUI.FortSocialAvatarIcon
   bInherited=true
   parent=Overlay_0
   widget=AvatarIcon
   properties {
    avatarMaterial=/Game/UI/SocialPanel/Art/Materials/UI_UserPortrait_Face
    materialTextureParamName=SmallPreview
    brush {
     bIsDynamicallyLoaded=false
     drawAs=Image
     imageType=NoImage
     mirroring=NoMirror
     resourceName=None
     resourceObject=&UI_UserPortrait_Face
     tiling=NoTile
     imageSize=<45 45>
     margin=<ltrb 0 0 0 0>
     outlineSettings {
      bUseBrushTransparency=false
      roundingType=HalfHeightRadius
      width=0
      color {colorUseRule=UseColor_Specified, specifiedColor=<rgba 0 0 0 0>}
      cornerRadii=<0 0 0 0>
     }
     tintColor {colorUseRule=UseColor_Specified, specifiedColor=<rgba 1 1 1 1>}
     uVRegion {bIsValid=false, max=<0 0>, min=<0 0>}
    }
   }
   slot {class=OverlaySlot, properties {padding=<ltrb 0 0 0 16>}}
  }
  Border_PlayerColor {
   widgetClassPath=&Border
   bIsVariable=true
   parent=Overlay_0
   widget=Border_PlayerColor
   properties {
    visibility=SelfHitTestInvisible
    background {
     bIsDynamicallyLoaded=false
     drawAs=NoDrawType
     imageType=NoImage
     mirroring=NoMirror
     resourceName=None
     resourceObject=null
     tiling=NoTile
     imageSize=<32 32>
     margin=<ltrb 0 0 0 0>
     outlineSettings {
      bUseBrushTransparency=false
      roundingType=HalfHeightRadius
      width=0
      color {colorUseRule=UseColor_Specified, specifiedColor=<rgba 0 0 0 0>}
      cornerRadii=<0 0 0 0>
     }
     tintColor {colorUseRule=UseColor_Specified, specifiedColor=<rgba 1 1 1 1>}
     uVRegion {bIsValid=false, max=<0 0>, min=<0 0>}
    }
    contentColorAndOpacity=<rgba 0 1 1 1>
    padding=<ltrb 0 0 0 0>
   }
   slot {class=OverlaySlot, properties {horizontalAlignment=HAlign_Fill, verticalAlignment=VAlign_Bottom}}
  }
  VerticalBox_20 {
   widgetClassPath=&VerticalBox
   parent=Border_PlayerColor
   widget=VerticalBox_20
   slot {class=BorderSlot, properties {padding=<ltrb 0 0 0 0>}}
  }
  Image_Line {
   widgetClassPath=&Image
   parent=VerticalBox_20
   widget=Image_Line
   properties {
    visibility=SelfHitTestInvisible
    brush {
     bIsDynamicallyLoaded=false
     drawAs=Image
     imageType=NoImage
     mirroring=NoMirror
     resourceName=None
     resourceObject=/Game/Athena/UI/ReplayBrowser/M_Ui_Skewed_Vfor9slices_top
     tiling=NoTile
     imageSize=<24.3159 5.7548>
     margin=<ltrb 0 0 0 0>
     outlineSettings {
      bUseBrushTransparency=false
      roundingType=HalfHeightRadius
      width=0
      color {colorUseRule=UseColor_Specified, specifiedColor=<rgba 0 0 0 0>}
      cornerRadii=<0 0 0 0>
     }
     tintColor {colorUseRule=UseColor_Specified, specifiedColor=<rgba 1 1 1 1>}
     uVRegion {bIsValid=false, max=<0 0>, min=<0 0>}
    }
   }
   slot {class=VerticalBoxSlot}
  }
  Image_101 {
   widgetClassPath=&Image
   parent=VerticalBox_20
   widget=Image_101
   properties {
    visibility=SelfHitTestInvisible
    brush {
     bIsDynamicallyLoaded=false
     drawAs=Image
     imageType=NoImage
     mirroring=NoMirror
     resourceName=None
     resourceObject=/Game/Athena/UI/ReplayBrowser/M_Ui_Skewed_triangle
     tiling=NoTile
     imageSize=<10 12>
     margin=<ltrb 0 0 0 0>
     outlineSettings {
      bUseBrushTransparency=false
      roundingType=HalfHeightRadius
      width=0
      color {colorUseRule=UseColor_Specified, specifiedColor=<rgba 0 0 0 0>}
      cornerRadii=<0 0 0 0>
     }
     tintColor {colorUseRule=UseColor_Specified, specifiedColor=<rgba 1 1 1 1>}
     uVRegion {bIsValid=false, max=<0 0>, min=<0 0>}
    }
   }
   slot {class=VerticalBoxSlot, properties {horizontalAlignment=HAlign_Center, verticalAlignment=VAlign_Center}}
  }
  SizeBox_26 {
   widgetClassPath=&SizeBox
   parent=Overlay_0
   widget=SizeBox_26
   properties {bOverride_HeightOverride=true, bOverride_WidthOverride=true, heightOverride=21.1781, widthOverride=1}
   slot {class=OverlaySlot, properties {horizontalAlignment=HAlign_Center, verticalAlignment=VAlign_Fill}}
  }
  VerticalBox_text {
   widgetClassPath=&VerticalBox
   bIsVariable=true
   parent=SizeBox_26
   widget=VerticalBox_text
   properties {renderTransform {angle=0, scale=<1 1>, shear=<0 0>, translation=<-32 0>}}
   slot {class=SizeBoxSlot}
  }
  Text_PlayerName {
   widgetClassPath=&CommonTextBlock
   bInherited=true
   bIsVariable=true
   parent=VerticalBox_text
   widget=Text_PlayerName
   properties {
    justification=Right
    lineHeightPercentage=0.9
    mobileFontSizeMultiplier=1.3
    text="{Player_Name999}"
    colorAndOpacity {colorUseRule=UseColor_Specified, specifiedColor=<rgba 0.1329 0.6038 1 1>}
    font {
     bForceMonospaced=false
     bMaterialIsStencil=false
     fontMaterial=null
     fontObject=&BurbankSmall
     letterSpacing=0
     monospacedWidth=1
     size=14
     skewAmount=0
     typefaceFontName=Black
     outlineSettings {
      bApplyOutlineToDropShadows=false
      bMiteredCorners=false
      bSeparateFillAlpha=false
      outlineMaterial=null
      outlineSize=0
      outlineColor=<rgba 0 0 0 1>
     }
    }
    renderTransform {angle=0, scale=<1 1>, shear=<-12 0>, translation=<0 0>}
    shadowOffset=<0 0>
   }
   slot {class=VerticalBoxSlot, properties {horizontalAlignment=HAlign_Right}}
  }
  RichText_Number {
   widgetClassPath=&CommonRichTextBlock
   bIsVariable=true
   parent=VerticalBox_text
   widget=RichText_Number
   properties {
    bOverrideDefaultStyle=true
    justification=Right
    mobileTextBlockScale=1.2
    text="{0}<text color=\"00a6ffff\" size=\"12\">/{00}</>"
    defaultTextStyleOverride {
     overflowPolicy=Clip
     transformPolicy=None
     colorAndOpacity {colorUseRule=UseColor_Specified, specifiedColor=<rgba 1 1 1 1>}
     font {
      bForceMonospaced=false
      bMaterialIsStencil=false
      fontMaterial=null
      fontObject=&BurbankSmall
      letterSpacing=0
      monospacedWidth=1
      size=16
      skewAmount=0
      typefaceFontName=Black
      outlineSettings {
       bApplyOutlineToDropShadows=false
       bMiteredCorners=false
       bSeparateFillAlpha=false
       outlineMaterial=null
       outlineSize=0
       outlineColor=<rgba 0 0 0 1>
      }
     }
     highlightColor {colorUseRule=UseColor_Specified, specifiedColor=<rgba 0 0 0 1>}
     highlightShape {
      bIsDynamicallyLoaded=false
      drawAs=Image
      imageType=NoImage
      mirroring=NoMirror
      resourceName=None
      resourceObject=null
      tiling=NoTile
      imageSize=<32 32>
      margin=<ltrb 0 0 0 0>
      outlineSettings {
       bUseBrushTransparency=false
       roundingType=HalfHeightRadius
       width=0
       color {colorUseRule=UseColor_Specified, specifiedColor=<rgba 0 0 0 0>}
       cornerRadii=<0 0 0 0>
      }
      tintColor {colorUseRule=UseColor_Specified, specifiedColor=<rgba 1 1 1 1>}
      uVRegion {bIsValid=false, max=<0 0>, min=<0 0>}
     }
     selectedBackgroundColor {colorUseRule=UseColor_Foreground, specifiedColor=<rgba 1 0 1 1>}
     shadowColorAndOpacity=<rgba 0 0 0 1>
     shadowOffset=<0 0>
     strikeBrush {
      bIsDynamicallyLoaded=false
      drawAs=Image
      imageType=NoImage
      mirroring=NoMirror
      resourceName=None
      resourceObject=null
      tiling=NoTile
      imageSize=<32 32>
      margin=<ltrb 0 0 0 0>
      outlineSettings {
       bUseBrushTransparency=false
       roundingType=HalfHeightRadius
       width=0
       color {colorUseRule=UseColor_Specified, specifiedColor=<rgba 0 0 0 0>}
       cornerRadii=<0 0 0 0>
      }
      tintColor {colorUseRule=UseColor_Specified, specifiedColor=<rgba 1 1 1 1>}
      uVRegion {bIsValid=false, max=<0 0>, min=<0 0>}
     }
     underlineBrush {
      bIsDynamicallyLoaded=false
      drawAs=Image
      imageType=NoImage
      mirroring=NoMirror
      resourceName=None
      resourceObject=null
      tiling=NoTile
      imageSize=<32 32>
      margin=<ltrb 0 0 0 0>
      outlineSettings {
       bUseBrushTransparency=false
       roundingType=HalfHeightRadius
       width=0
       color {colorUseRule=UseColor_Specified, specifiedColor=<rgba 0 0 0 0>}
       cornerRadii=<0 0 0 0>
      }
      tintColor {colorUseRule=UseColor_Specified, specifiedColor=<rgba 1 1 1 1>}
      uVRegion {bIsValid=false, max=<0 0>, min=<0 0>}
     }
    }
    renderTransform {angle=0, scale=<1 1>, shear=<-12 0>, translation=<0 0>}
   }
   slot {class=VerticalBoxSlot}
  }
  Text_Progress {
   widgetClassPath=&CommonTextBlock
   bInherited=true
   bIsVariable=true
   parent=VerticalBox_text
   widget=Text_Progress
   properties {
    justification=Right
    lineHeightPercentage=0.9
    mobileFontSizeMultiplier=1.2
    text=0
    visibility=Collapsed
    font {
     bForceMonospaced=false
     bMaterialIsStencil=false
     fontMaterial=null
     fontObject=&BurbankSmall
     letterSpacing=0
     monospacedWidth=1
     size=14
     skewAmount=0
     typefaceFontName=Black
     outlineSettings {
      bApplyOutlineToDropShadows=false
      bMiteredCorners=false
      bSeparateFillAlpha=false
      outlineMaterial=null
      outlineSize=0
      outlineColor=<rgba 0 0 0 1>
     }
    }
    renderTransform {angle=0, scale=<1 1>, shear=<-12 0>, translation=<0 0>}
    shadowOffset=<0 0>
   }
   slot {class=VerticalBoxSlot, properties {horizontalAlignment=HAlign_Center, verticalAlignment=VAlign_Center}}
  }
  Image_Arrow {
   widgetClassPath=&Image
   parent=VerticalBox_17
   widget=Image_Arrow
   properties {
    brush {
     bIsDynamicallyLoaded=false
     drawAs=Image
     imageType=NoImage
     mirroring=NoMirror
     resourceName=None
     resourceObject=/Game/Athena/UI/ReplayBrowser/M_Ui_Skewed_triangle_collection
     tiling=NoTile
     imageSize=<32 32>
     margin=<ltrb 0 0 0 0>
     outlineSettings {
      bUseBrushTransparency=false
      roundingType=HalfHeightRadius
      width=0
      color {colorUseRule=UseColor_Specified, specifiedColor=<rgba 0 0 0 0>}
      cornerRadii=<0 0 0 0>
     }
     tintColor {colorUseRule=UseColor_Specified, specifiedColor=<rgba 1 1 1 1>}
     uVRegion {bIsValid=false, max=<0 0>, min=<0 0>}
    }
    renderTransform {angle=-90, scale=<0.5 0.5>, shear=<0 0>, translation=<0 0>}
   }
   slot {class=VerticalBoxSlot, properties {horizontalAlignment=HAlign_Center, verticalAlignment=VAlign_Center}}
  }
  SizeBox_LocalPlayer {
   widgetClassPath=&SizeBox
   bIsVariable=true
   parent=WidgetSwitcher_0
   widget=SizeBox_LocalPlayer
   properties {bOverride_WidthOverride=true}
   slot {class=WidgetSwitcherSlot, properties {horizontalAlignment=HAlign_Right, verticalAlignment=VAlign_Bottom}}
  }
  VerticalBox {widgetClassPath=&VerticalBox, parent=SizeBox_LocalPlayer, widget=VerticalBox, slot {class=SizeBoxSlot}}
  Image_Arrow_1 {
   widgetClassPath=&Image
   parent=VerticalBox
   widget=Image_Arrow_1
   properties {
    brush {
     bIsDynamicallyLoaded=false
     drawAs=Image
     imageType=NoImage
     mirroring=NoMirror
     resourceName=None
     resourceObject=&M_Ui_Skewed_triangle_collection
     tiling=NoTile
     imageSize=<16 16>
     margin=<ltrb 0 0 0 0>
     outlineSettings {
      bUseBrushTransparency=false
      roundingType=HalfHeightRadius
      width=0
      color {colorUseRule=UseColor_Specified, specifiedColor=<rgba 0 0 0 0>}
      cornerRadii=<0 0 0 0>
     }
     tintColor {colorUseRule=UseColor_Specified, specifiedColor=<rgba 0.8714 1 0 1>}
     uVRegion {bIsValid=false, max=<0 0>, min=<0 0>}
    }
    renderTransform {angle=0, scale=<-1 -1>, shear=<0 0>, translation=<-1 0>}
   }
   slot {class=VerticalBoxSlot, properties {horizontalAlignment=HAlign_Center, padding=<ltrb -8 -3 -8 0>}}
  }
  RichText_Number_Local {
   widgetClassPath=&CommonRichTextBlock
   bIsVariable=true
   parent=VerticalBox
   widget=RichText_Number_Local
   properties {
    bOverrideDefaultStyle=true
    justification=Center
    mobileTextBlockScale=1.4
    text="{0}<text color=\"00a6ffff\" size=\"12\">/{00}</>"
    defaultTextStyleOverride {
     overflowPolicy=Clip
     transformPolicy=None
     colorAndOpacity {colorUseRule=UseColor_Specified, specifiedColor=<rgba 0.8714 1 0 1>}
     font {
      bForceMonospaced=false
      bMaterialIsStencil=false
      fontMaterial=null
      fontObject=&BurbankSmall
      letterSpacing=0
      monospacedWidth=1
      size=14
      skewAmount=0
      typefaceFontName=Black
      outlineSettings {
       bApplyOutlineToDropShadows=false
       bMiteredCorners=false
       bSeparateFillAlpha=false
       outlineMaterial=null
       outlineSize=0
       outlineColor=<rgba 0 0 0 1>
      }
     }
     highlightColor {colorUseRule=UseColor_Specified, specifiedColor=<rgba 0 0 0 1>}
     highlightShape {
      bIsDynamicallyLoaded=false
      drawAs=Image
      imageType=NoImage
      mirroring=NoMirror
      resourceName=None
      resourceObject=null
      tiling=NoTile
      imageSize=<32 32>
      margin=<ltrb 0 0 0 0>
      outlineSettings {
       bUseBrushTransparency=false
       roundingType=HalfHeightRadius
       width=0
       color {colorUseRule=UseColor_Specified, specifiedColor=<rgba 0 0 0 0>}
       cornerRadii=<0 0 0 0>
      }
      tintColor {colorUseRule=UseColor_Specified, specifiedColor=<rgba 1 1 1 1>}
      uVRegion {bIsValid=false, max=<0 0>, min=<0 0>}
     }
     selectedBackgroundColor {colorUseRule=UseColor_Foreground, specifiedColor=<rgba 1 0 1 1>}
     shadowColorAndOpacity=<rgba 0 0 0 1>
     shadowOffset=<0 0>
     strikeBrush {
      bIsDynamicallyLoaded=false
      drawAs=Image
      imageType=NoImage
      mirroring=NoMirror
      resourceName=None
      resourceObject=null
      tiling=NoTile
      imageSize=<32 32>
      margin=<ltrb 0 0 0 0>
      outlineSettings {
       bUseBrushTransparency=false
       roundingType=HalfHeightRadius
       width=0
       color {colorUseRule=UseColor_Specified, specifiedColor=<rgba 0 0 0 0>}
       cornerRadii=<0 0 0 0>
      }
      tintColor {colorUseRule=UseColor_Specified, specifiedColor=<rgba 1 1 1 1>}
      uVRegion {bIsValid=false, max=<0 0>, min=<0 0>}
     }
     underlineBrush {
      bIsDynamicallyLoaded=false
      drawAs=Image
      imageType=NoImage
      mirroring=NoMirror
      resourceName=None
      resourceObject=null
      tiling=NoTile
      imageSize=<32 32>
      margin=<ltrb 0 0 0 0>
      outlineSettings {
       bUseBrushTransparency=false
       roundingType=HalfHeightRadius
       width=0
       color {colorUseRule=UseColor_Specified, specifiedColor=<rgba 0 0 0 0>}
       cornerRadii=<0 0 0 0>
      }
      tintColor {colorUseRule=UseColor_Specified, specifiedColor=<rgba 1 1 1 1>}
      uVRegion {bIsValid=false, max=<0 0>, min=<0 0>}
     }
    }
    renderTransform {angle=0, scale=<1 1>, shear=<-12 0>, translation=<0 0>}
   }
   slot {class=VerticalBoxSlot, properties {horizontalAlignment=HAlign_Center}}
  }
 ]
}
