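# Widget for a collection progress track (Fishing by default) in Fortnite's UI, based on AthenaCollectionPlayerProgressContainer, comparing players' collected counts along a sheared bar.
# When entries are created, it adds each to a one-row uniform grid at its collected count minus one.
# An invisible spacer image, re-added on reset, moves to the last column to set the grid's width.
# The local player's ratio drives the track material's Progress parameter; an Intro animation slides and fades the title.
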WidgetBlueprint /FortniteCoreUI/Collections/BP_AthenaCollectionPlayerProgressTrack
animations [Intro {
  binding_count=1
  length_seconds=0.45
  object=&self:Intro
  sequence {bindings [Text_Title {
     id=#1
     possessed_class=CommonTextBlock
     tracks [
      {
       type=MovieScene2DTransformTrack
       sections [{
         type=MovieScene2DTransformSection
         channels [
          Translation.X {type=MovieSceneScriptingFloatChannel, keys [[0,64],[9,0]]}
          Translation.Y {type=MovieSceneScriptingFloatChannel}
         ]
        }]
      }
      {
       type=MovieSceneFloatTrack
       display_name=Render Opacity
       sections [{type=MovieSceneFloatSection, channels [None {type=MovieSceneScriptingFloatChannel, keys [[0,0],[9,1]]}]}]
      }
     ]
    }]}
 }]
graphs [EventGraph {nodes [
   PlayIntro() {is=title K2Node_CustomEvent}
   > PlayAnimationForward(InAnimation=Get Intro())
   Event Construct()
   > n1 = GetDynamicMaterial(self=Get Image_RaceTrack() -> (Image_RaceTrack:Image Object Reference))
   > SetScalarParameterValue(self=@n1, ParameterName=Progress, Value=0)
   Event On Player Progress Container Reset() {is=title K2Node_Event}
   n2 = Get GridPanel_Friends() -> (GridPanel_Friends:Uniform Grid Panel Object Reference)
   > ClearChildren(self=@n2)
   > n3 = AddChildToUniformGrid!(
    execute
    self:Uniform Grid Panel Object Reference=@n2
    Content:Widget Object Reference=Get Image_SetMaxnumber() -> (Image_SetMaxnumber:Image Object Reference)
    InRow:int=0
    InColumn:int=0
   ) -> (then, ReturnValue:Uniform Grid Slot Object Reference)
   > SetHorizontalAlignment(self:Uniform Grid Slot Object Reference=@n3, InHorizontalAlignment=HAlign_Fill)
   > SetVerticalAlignment(self:Uniform Grid Slot Object Reference=@n3, InVerticalAlignment=VAlign_Fill)
   Event PreConstruct()
   > "SetText (Text)"(self=Get Text_Title() -> (Text_Title:Common Text Object Reference), InText=Get Title Text() -> (Title Text:Text))
   n4 = Event On Player Progress Entries Created!() -> (OutputDelegate, then, PlayerProgressEntries:Array of Athena Collection Player Progress Object References, NumToCollect:int=0) {is=title K2Node_Event}
   > SetVisibility(self=Get GridPanel_Friends(), InVisibility=SelfHitTestInvisible)
   n5 = SlotAsUniformGridSlot!(Widget:Widget Object Reference=Get Image_SetMaxnumber()) -> (ReturnValue:Uniform Grid Slot Object Reference)
   > n6 = Is Valid(InputObject=@n5)
   n7 = For Each Loop(Exec=[@n12.then, @n6.Is Not Valid], Array:Array of Athena Collection Player Progress Object References=@n4.PlayerProgressEntries) -> (Array Element:Athena Collection Player Progress Object Reference)
   > n8 = AddChildToUniformGrid(
    self=Get GridPanel_Friends()
    Content=@n7
    InRow=0
    InColumn=int - int(A=GetNumCollected!(self:Athena Collection Player Progress Object Reference=@n7) -> (ReturnValue:int=0), B=1) {is=title K2Node_CallFunction}
   )
   > SetHorizontalAlignment(self=@n8, InHorizontalAlignment=HAlign_Fill)
   > SetVerticalAlignment(self=@n8, InVerticalAlignment=VAlign_Fill)
   > Branch(Condition=IsLocalPlayer!(self:Athena Collection Player Progress Object Reference=@n7) -> (ReturnValue:bool))
   > n9 = GetDynamicMaterial(self=Get Image_RaceTrack())
   n10 = GetNumCollected(self=@n7) -> (ReturnValue=0)
   n11 = Select(Option 0:int=@n4.NumToCollect, Option 1:int=1, Index="Equal (Integer)"(A=@n4.NumToCollect)) -> (ReturnValue:int)
   > SetScalarParameterValue(
    self=@n9
    ParameterName=Progress
    Value=float / float(A="To Float (Integer)"(InInt=@n10), B="To Float (Integer)"(InInt=@n11)) {is=title K2Node_CallFunction}
   )
   n12 = SetColumn!(execute=@n6.Is Valid, self:Uniform Grid Slot Object Reference=@n5, InColumn:int=int - int(A=@n4.NumToCollect, B=1)) -> (then)
  ]}]
variables [Title Text {type=string, default=Fishing}]
widgets {
 info {
  inheritedWidgetCount=0
  parentClass=/Script/FortniteUI.AthenaCollectionPlayerProgressContainer
  rootWidgetClass=&SizeBox
  widgetCount=8
 }
 widgets [
  SizeBox_0 {widgetClassPath=&SizeBox, widget=SizeBox_0, properties {bOverride_WidthOverride=true, heightOverride=116, widthOverride=350}}
  Overlay_1 {widgetClassPath=&Overlay, parent=SizeBox_0, widget=Overlay_1, slot {class=SizeBoxSlot}}
  Text_Title {
   widgetClassPath=&CommonTextBlock
   bIsVariable=true
   parent=Overlay_1
   widget=Text_Title
   properties {
    text=FISHING
    textTransformPolicy=ToUpper
    colorAndOpacity {colorUseRule=UseColor_Specified, specifiedColor=<rgba 0 0.0908 1 0.5>}
    font {
     bForceMonospaced=false
     bMaterialIsStencil=false
     fontMaterial=/FortniteCoreUI/UI/Season/Art/M_UI_AthenaSeason_BattlePass_Text_white2
     fontObject=&Burbank
     letterSpacing=0
     monospacedWidth=1
     size=42
     skewAmount=0
     typefaceFontName=BurbankRegular-Black
     outlineSettings {
      bApplyOutlineToDropShadows=false
      bMiteredCorners=false
      bSeparateFillAlpha=false
      outlineMaterial=null
      outlineSize=0
      outlineColor=<rgba 0 0 0 1>
     }
    }
    renderTransform {angle=0, scale=<1 1>, shear=<-10 0>, translation=<0 0>}
    shadowOffset=<0 0>
   }
   slot {class=OverlaySlot, properties {horizontalAlignment=HAlign_Fill, padding=<ltrb 4 0 0 0>}}
  }
  Image_RaceTrack {
   widgetClassPath=&Image
   bIsVariable=true
   parent=Overlay_1
   widget=Image_RaceTrack
   properties {
    visibility=SelfHitTestInvisible
    brush {
     bIsDynamicallyLoaded=false
     drawAs=Box
     imageType=NoImage
     mirroring=NoMirror
     resourceName=None
     resourceObject=/FortniteCoreUI/Athena/Collections/Art/M_Ui_Collection_Track_Inst
     tiling=NoTile
     imageSize=<32 28>
     margin=<ltrb 0.5 0 0.5 0>
     outlineSettings {
      bUseBrushTransparency=false
      roundingType=HalfHeightRadius
      width=0
      color {colorUseRule=UseColor_Specified, specifiedColor=<rgba 0 0 0 0>}
      cornerRadii=<0 0 0 0>
     }
     tintColor {colorUseRule=UseColor_Specified, specifiedColor=<rgba 1 1 1 1>}
     uVRegion {bIsValid=false, max=<0 0>, min=<0 0>}
    }
    renderTransform {angle=0, scale=<1 1>, shear=<-10 0>, translation=<0 0>}
   }
   slot {class=OverlaySlot, properties {horizontalAlignment=HAlign_Fill, padding=<ltrb 0 57 0 0>}}
  }
  GridPanel_Friends {
   widgetClassPath=/Script/UMG.UniformGridPanel
   bIsVariable=true
   parent=Overlay_1
   widget=GridPanel_Friends
   properties {visibility=Collapsed}
   slot {class=OverlaySlot, properties {horizontalAlignment=HAlign_Fill}}
  }
  BP_AthenaCollection_ProgressPlayerEntry {
   widgetClassPath=/Game/Athena/Collections/BP_AthenaCollection_ProgressPlayerEntry
   parent=GridPanel_Friends
   widget=BP_AthenaCollection_ProgressPlayerEntry
   properties {is Self pre construct=true}
   slot {class=UniformGridSlot, properties {column=9, horizontalAlignment=HAlign_Fill, verticalAlignment=VAlign_Bottom}}
  }
  BP_AthenaCollection_ProgressPlayerEntry_1 {
   widgetClassPath=&BP_AthenaCollection_ProgressPlayerEntry
   parent=GridPanel_Friends
   widget=BP_AthenaCollection_ProgressPlayerEntry_1
   slot {class=UniformGridSlot, properties {column=13, horizontalAlignment=HAlign_Fill, verticalAlignment=VAlign_Bottom}}
  }
  Image_SetMaxnumber {
   widgetClassPath=&Image
   bIsVariable=true
   parent=GridPanel_Friends
   widget=Image_SetMaxnumber
   properties {
    visibility=SelfHitTestInvisible
    brush {
     bIsDynamicallyLoaded=false
     drawAs=NoDrawType
     imageType=NoImage
     mirroring=NoMirror
     resourceName=None
     resourceObject=null
     tiling=NoTile
     imageSize=<1 1>
     margin=<ltrb 0 0 0 0>
     outlineSettings {
      bUseBrushTransparency=false
      roundingType=HalfHeightRadius
      width=0
      color {colorUseRule=UseColor_Specified, specifiedColor=<rgba 0 0 0 0>}
      cornerRadii=<0 0 0 0>
     }
     tintColor {colorUseRule=UseColor_Specified, specifiedColor=<rgba 1 1 1 1>}
     uVRegion {bIsValid=false, max=<0 0>, min=<0 0>}
    }
   }
   slot {class=UniformGridSlot, properties {column=40, horizontalAlignment=HAlign_Fill, verticalAlignment=VAlign_Fill}}
  }
 ]
}
